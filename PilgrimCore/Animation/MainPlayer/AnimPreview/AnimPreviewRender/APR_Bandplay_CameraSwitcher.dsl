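; Debug-only preview actor that stages the player's bandplay animation and cameras for review when the player enters its trigger box.
; On overlap it plays a level sequence, loops the animation on the player mesh and attaches FaceCam to the Head socket.
; RotateTimeline then drives the world yaw of CamOffset.
; BeginPlay enables player input, but the J, K, R and P key events are empty.
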(event OnComponentBeginOverlap(TriggerBox)
  (params (OverlappedComponent :type Primitive%20Component%20Object%20Reference) (OtherActor :type Actor%20Object%20Reference) (OtherComp :type Primitive%20Component%20Object%20Reference) (OtherBodyIndex :type Integer) (bFromSweep :type Boolean) (SweepResult :type Hit%20Result%20Structure%20%28by%20ref%29))
  (bind _returnvalue (Game|GetPlayerCharacter 0))
  (bind _facecam (Variables|Default|GetFaceCam))
  ;[1 0.4688 0.4688 1 60] FOR DEBUGGING/ REVIEWING ONLY
  (if (== _returnvalue OtherActor)
    (bind (_outactor _returnvalue_2) (Sequencer|Player|CreateLevelSequencePlayer (Variables|Default|GetLS)))
    (Sequencer|Player|Play _returnvalue_2 :!class /Script/MovieScene.MovieSceneSequencePlayer)
    (Components|Animation|PlayAnimation (Variables|/Script/Engine.Character|GetMesh (Game|GetPlayerCharacter 0)) (Variables|Default|GetBandplay_Animation) true)
    (Transformation|AttachComponentToComponent _facecam (Variables|/Script/Engine.Character|GetMesh _returnvalue) "Head" "KeepRelative" "KeepRelative" "KeepRelative" true)
    (Transformation|SetRelativeLocationAndRotation _facecam "0.000000,30.000000,0.000000" "180.000000,90.000000,-90.000000" false false)
    (bind (_direction _camerakeys) (Timeline|RotateTimeline Play)
      (:Update
        (Transformation|SetWorldRotation (Variables|Default|GetCamOffset) (Math|Rotator|MakeRotator 0.0 0.0 _camerakeys) false false)))))

(event EventBeginPlay
  (Input|EnableInput (Game|GetPlayerController 0)))

(event J
  (params (Key :type Key%20Structure)))

(event K
  (params (Key :type Key%20Structure)))

(event R
  (params (Key :type Key%20Structure)))

(event P
  (params (Key :type Key%20Structure)))

(fn ConstructionScript)
