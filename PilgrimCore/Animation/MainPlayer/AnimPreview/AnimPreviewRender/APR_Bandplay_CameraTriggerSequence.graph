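# Level Sequence for the Bandplay animation preview render, providing a cine camera setup under a single camera cut section.
# A spawned CineCameraActor holds a fixed transform and constant focal length, focus distance and aperture.
# A scaled Image Plate and a plain Actor are also spawned, and all three actors carry attach tracks, with targets not shown.
# The APR_Bandplay_CameraSwitcher Blueprint is bound without tracks, and no event track appears despite the trigger name.
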LevelSequence /PilgrimCore/Animation/MainPlayer/AnimPreview/AnimPreviewRender/APR_Bandplay_CameraTriggerSequence
bindings [
 CameraComponent {
  id=#1
  parent_id=#2
  possessed_class=CineCameraComponent
  tracks [
   {
    type=MovieSceneFloatTrack
    display_name=Current Focal Length
    sections [{type=MovieSceneFloatSection, channels [None {type=MovieSceneScriptingFloatChannel, default=35}]}]
   }
   {
    type=MovieSceneFloatTrack
    display_name="Manual Focus Distance (Focus Settings)"
    sections [{type=MovieSceneFloatSection, channels [None {type=MovieSceneScriptingFloatChannel, default=650}]}]
   }
   {
    type=MovieSceneFloatTrack
    display_name=Current Aperture
    sections [{type=MovieSceneFloatSection, channels [None {type=MovieSceneScriptingFloatChannel, default=2.8}]}]
   }
  ]
 }
 Cine Camera Actor {
  id=#2
  possessed_class=CineCameraActor
  tracks [
   {
    type=MovieScene3DTransformTrack
    sections [{
      type=MovieScene3DTransformSection
      channels [
       Location.X {type=MovieSceneScriptingDoubleChannel, default=14.1253}
       Location.Y {type=MovieSceneScriptingDoubleChannel, default=-2179.1242}
       Location.Z {type=MovieSceneScriptingDoubleChannel, default=270}
       Rotation.X {type=MovieSceneScriptingDoubleChannel, default=0}
       Rotation.Y {type=MovieSceneScriptingDoubleChannel, default=-11.2}
       Rotation.Z {type=MovieSceneScriptingDoubleChannel, default=91}
       Scale.X {type=MovieSceneScriptingDoubleChannel, default=1}
       Scale.Y {type=MovieSceneScriptingDoubleChannel, default=1}
       Scale.Z {type=MovieSceneScriptingDoubleChannel, default=1}
      ]
     }]
   }
   {type=MovieSceneSpawnTrack, sections [{type=MovieSceneSpawnSection, channels [None {type=MovieSceneScriptingBoolChannel, default=true}]}]}
   {type=MovieScene3DAttachTrack, sections [{type=MovieScene3DAttachSection, end_frame=9910, start_frame=0}]}
  ]
 }
 Image Plate {
  id=#3
  possessed_class=ImagePlate
  tracks [
   {type=MovieSceneSpawnTrack, sections [{type=MovieSceneSpawnSection, channels [None {type=MovieSceneScriptingBoolChannel, default=true}]}]}
   {
    type=MovieScene3DTransformTrack
    sections [{
      type=MovieScene3DTransformSection
      channels [
       Location.X {type=MovieSceneScriptingDoubleChannel, default=180.1441}
       Location.Y {type=MovieSceneScriptingDoubleChannel, default=0}
       Location.Z {type=MovieSceneScriptingDoubleChannel, default=0}
       Rotation.X {type=MovieSceneScriptingDoubleChannel, default=0}
       Rotation.Y {type=MovieSceneScriptingDoubleChannel, default=0}
       Rotation.Z {type=MovieSceneScriptingDoubleChannel, default=0}
       Scale.X {type=MovieSceneScriptingDoubleChannel, default=1.5}
       Scale.Y {type=MovieSceneScriptingDoubleChannel, default=61.1461}
       Scale.Z {type=MovieSceneScriptingDoubleChannel, default=34.3947}
      ]
     }]
   }
   {type=MovieScene3DAttachTrack, sections [{type=MovieScene3DAttachSection, end_frame=9910, start_frame=0}]}
  ]
 }
 APR_Bandplay_CameraSwitcher {id=#4, possessed_class=APR_Bandplay_CameraSwitcher_C, tracks []}
 Actor {
  id=#5
  possessed_class=Actor
  tracks [
   {type=MovieSceneSpawnTrack, sections [{type=MovieSceneSpawnSection, channels [None {type=MovieSceneScriptingBoolChannel, default=true}]}]}
   {
    type=MovieScene3DTransformTrack
    sections [{
      type=MovieScene3DTransformSection
      channels [
       Location.X {type=MovieSceneScriptingDoubleChannel, default=0}
       Location.Y {type=MovieSceneScriptingDoubleChannel, default=0}
       Location.Z {type=MovieSceneScriptingDoubleChannel, default=0}
       Rotation.X {type=MovieSceneScriptingDoubleChannel, default=0}
       Rotation.Y {type=MovieSceneScriptingDoubleChannel, default=0}
       Rotation.Z {type=MovieSceneScriptingDoubleChannel, default=360}
       Scale.X {type=MovieSceneScriptingDoubleChannel, default=1}
       Scale.Y {type=MovieSceneScriptingDoubleChannel, default=1}
       Scale.Z {type=MovieSceneScriptingDoubleChannel, default=1}
      ]
     }]
   }
   {type=MovieScene3DAttachTrack, sections [{type=MovieScene3DAttachSection, end_frame=9910, start_frame=0}]}
  ]
 }
]
display_rate {denominator=1, numerator=30}
tick_resolution {denominator=1, numerator=24000}
tracks [{type=MovieSceneCameraCutTrack, sections [{type=MovieSceneCameraCutSection, end_frame=9911, start_frame=0}]}]
