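# UI material that composites two Bandplay render targets into one image, apparently for the main player animation preview.
# RT1 is sampled with horizontally tiled and shifted UVs, RT2 with a horizontal shift only.
# The green and red channels of a shared mask texture weight RT1 and RT2 respectively, and the results are added together.
# The material uses alpha composite blending.
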Material /PilgrimCore/Animation/MainPlayer/AnimPreview/AnimPreviewRender/MAT_Bandplay_renderTarget
graph {
 properties {blendMode=BLEND_AlphaComposite, materialDomain=MD_UI, pixelDepthOffsetMode=PDOM_Legacy, usedShadingModels=Default Lit}
 nodes [
  n1 = MaterialFunctionCall(Float2=TextureCoordinate() {uTiling=2}) {materialFunction=&BreakOutFloat2Components}
  n2 = MaterialFunctionCall(X=Add(A=@n1.R) {constB=-0.25}, Y=@n1.G) {materialFunction=&MakeFloat2}
  n3 = TextureSample() {texture=&/TEX_Bandplay_Render_Mask}
  n4 = MaterialFunctionCall(Float2=TextureCoordinate()) {materialFunction=&BreakOutFloat2Components}
  n5 = MaterialFunctionCall(X=Add(A=@n4.R) {constB=0.125}, Y=@n4.G) {materialFunction=&MakeFloat2}
  Add(
   A=Multiply(
    A=TextureSample(UVs=@n2.Result) {samplerType=SAMPLERTYPE_LinearColor, texture=&/APR_Bandplay_RT1}.RGB
    B=@n3.G
   )
   B=Multiply(
    A=TextureSample(UVs=@n5.Result) {samplerType=SAMPLERTYPE_LinearColor, texture=&/APR_Bandplay_RT2}.RGB
    B=@n3.R
   )
  )
 ]
}
